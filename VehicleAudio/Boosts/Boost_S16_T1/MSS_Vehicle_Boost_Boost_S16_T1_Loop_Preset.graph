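# MetaSound Source preset of MSS_Vehicle_Boost_Loop that supplies the stereo looping sound for the Boost_S16_T1 vehicle boost.
# It feeds one loop wave, with fade, delay and pitch-range settings, into the parent graph; the one-shot wave array is empty.
# The source exposes Fort.Vehicle interface inputs and routes to the exterior boost sound class, with shared attenuation, concurrency, and volume and filter modulation.
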MetaSoundSource /VehicleAudio/Boosts/Boost_S16_T1/MSS_Vehicle_Boost_Boost_S16_T1_Loop_Preset
preset_parent=/VehicleAudio/Boosts/_ParentMetasound/MSS_Vehicle_Boost_Loop
inputs [
 Boost.Loop.DelayTime {type=Time, access=Constructor}
 Boost.Loop.FadeTime {type=Time, access=Constructor}
 Boost.Loop.One Shot.TransitionType {type=Int32}
 Boost.Loop.OneShot.CrossfadeTime {type=Time, access=Constructor}
 Boost.Loop.OneShot.RandomPitch.Max {type=Float, access=Constructor}
 Boost.Loop.OneShot.RandomPitch.Min {type=Float, access=Constructor}
 Boost.Loop.OneShot.Stereo {type=Bool}
 Boost.Loop.OneShot.TriggerRate.Max {type=Time, access=Constructor}
 Boost.Loop.OneShot.TriggerRate.Min {type=Time, access=Constructor}
 Boost.Loop.OneShot.Volume {type=Float, access=Constructor}
 Boost.Loop.Oneshots {type=WaveAsset:Array, access=Constructor}
 Boost.Loop.StartTime.Defined {type=Time, access=Constructor}
 Boost.Loop.StartTime.Random {type=Bool}
 Boost.Loop.Volume {type=Float, access=Constructor}
 Boost.Loop.Waveasset {type=WaveAsset:Array, access=Constructor}
 Boost.Pitch.Max {type=Float, access=Constructor}
 Boost.Pitch.Min {type=Float, access=Constructor}
 Boost.PitchByNormalSpeed {type=Bool, access=Constructor}
 Fort.Vehicle.Acceleration {type=Float}
 Fort.Vehicle.BlowoffStrength {type=Float}
 Fort.Vehicle.BonusSpeed {type=Float}
 Fort.Vehicle.EnableShifting {type=Bool}
 Fort.Vehicle.Gear {type=Int32}
 Fort.Vehicle.IsGrounded {type=Bool}
 Fort.Vehicle.Load {type=Float}
 Fort.Vehicle.LowFuel {type=Bool}
 Fort.Vehicle.NormalSpeed {type=Float}
 Fort.Vehicle.OnBlowoff {type=Trigger}
 Fort.Vehicle.OnDownshift {type=Trigger}
 Fort.Vehicle.OnThrottleEngaged {type=Trigger}
 Fort.Vehicle.OnThrottleReleased {type=Trigger}
 Fort.Vehicle.OnUpshift {type=Trigger}
 Fort.Vehicle.Pitch {type=Float}
 Fort.Vehicle.RPM {type=Float}
 Fort.Vehicle.Throttle {type=Float}
 UE.Source.OnPlay {type=Trigger}
 UE.Source.Orientation.Azimuth {type=Float}
 UE.Source.Orientation.Elevation {type=Float}
]
interfaces [Fort.Vehicle:1.0, UE.OutputFormat.Stereo:1.0, UE.Source.Orientation:1.0, UE.Source:1.1]
outputs [
 "UE.OutputFormat.Stereo.Audio:0" {type=Audio}
 "UE.OutputFormat.Stereo.Audio:1" {type=Audio}
]
pages [Default {input_defaults {
   Boost.Loop.DelayTime=0.1
   Boost.Loop.FadeTime=0.3
   Boost.Loop.One Shot.TransitionType=1
   Boost.Loop.OneShot.CrossfadeTime=1
   Boost.Loop.OneShot.RandomPitch.Max=2
   Boost.Loop.OneShot.RandomPitch.Min=-1
   Boost.Loop.OneShot.Stereo="<MetaSoundBuilderResult.SUCCEEDED: 0>"
   Boost.Loop.OneShot.TriggerRate.Max=0.1
   Boost.Loop.OneShot.TriggerRate.Min=0.216
   Boost.Loop.OneShot.Volume=0
   Boost.Loop.StartTime.Defined=0
   Boost.Loop.StartTime.Random="<MetaSoundBuilderResult.SUCCEEDED: 0>"
   Boost.Loop.Volume=0
   Boost.Pitch.Max=2.54
   Boost.Pitch.Min=0
   Boost.PitchByNormalSpeed="<MetaSoundBuilderResult.SUCCEEDED: 0>"
   Fort.Vehicle.Acceleration=0
   Fort.Vehicle.BlowoffStrength=0
   Fort.Vehicle.BonusSpeed=0
   Fort.Vehicle.EnableShifting=null
   Fort.Vehicle.Gear=0
   Fort.Vehicle.IsGrounded="<MetaSoundBuilderResult.SUCCEEDED: 0>"
   Fort.Vehicle.Load=0
   Fort.Vehicle.LowFuel=null
   Fort.Vehicle.NormalSpeed=0
   Fort.Vehicle.OnBlowoff=null
   Fort.Vehicle.OnDownshift=null
   Fort.Vehicle.OnThrottleEngaged=null
   Fort.Vehicle.OnThrottleReleased=null
   Fort.Vehicle.OnUpshift=null
   Fort.Vehicle.Pitch=0
   Fort.Vehicle.RPM=0
   Fort.Vehicle.Throttle=0
   UE.Source.OnPlay=null
   UE.Source.Orientation.Azimuth=0
   UE.Source.Orientation.Elevation=0
   Boost.Loop.Oneshots []
   Boost.Loop.Waveasset [&/Boost_S16_T1_Loop_01]
  }}]
properties {
 assetImportData=&self:AssetImportData
 attenuationSettings=/VehicleAudio/Boosts/Vehicle_Boost_Att
 audioPropertiesSheet=/VehicleAudio/PropertySheets/APS_Cosmetic_Vehicle_Boost_Loop
 bLooping=true
 duration=10000
 maxDistance=5012
 numChannels=2
 outputFormat=Stereo
 soundClassObject=/Game/Sounds/FortSoundClassesAndModes/Fort_Vehicle_Exterior_Boost
 soundSubmixObject=&Fortnite_SFX
 concurrencySet [/VehicleAudio/Boosts/Vehicle_Boost_Loop_Conc]
 modulationSettings {
  highpassRouting=Union
  lowpassRouting=Union
  volumeRouting=Union
  highpassModulationDestination {bEnableModulation=true, modulators [/Game/Sounds/Modulation/Patches/HPF/PP_FortVehicles_Boost_HPF]}
  lowpassModulationDestination {bEnableModulation=true, modulators [/Game/Sounds/Modulation/Patches/LPF/PP_FortVehicles_Boost_LPF]}
  volumeModulationDestination {
   bEnableModulation=true
   modulators [/VehicleAudio/Modulation/DM_VehicleTurboPatch, /Game/Sounds/Modulation/Patches/PP_FortVehicles_Boost_Vol]
  }
 }
}
